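# Niagara swing-trail effect for the Ice variant of the Crew pickaxe cosmetic template, pairing cyan ribbons with falling ice flakes.
# TrailIce and Trail_Bright_Blue are anim-trail ribbons, a wide base and a thin brighter core, that narrow and fade over their short lives.
# Sprite_Ice_Flakes spawns small sprites around socket-reader positions with interpolated sampling, then drifts them with gravity, drag and curl noise.
# Time-of-day scalars dim all three emitters' alpha; flakes also use distance scalability.
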NiagaraSystem /BRCosmetics/Effects/Niagara/Systems/CosmeticTemplates/Pickaxe_Impacts/Crew/Ice/NS_Pickaxe_Crew_Trail_Ice
dependencies {
 usedDataInterfaces [&NiagaraDataInterfaceCurve, &NiagaraDataInterfaceSocketReader]
 usedDynamicInputs []
 usedModules [
  &self:SystemSpawnScript.SystemScriptSource.SystemScriptGraph.NiagaraNodeAssignment_0.SetVariables_#1
  &SystemState
 ]
}
emitters {
 Sprite_Ice_Flakes {
  emitterUpdateScript [
   &EmitterState(
    Enable Distance Culling=true
    Enable Visibility Culling=true
    Life Cycle Mode=::NewEnumerator1
    Loop Behavior=::NewEnumerator1
    Loop Duration=2
    MaxDistance=2500
    Module.bApplySpawnCountScaleByDistance=true
    Module.bUseMaxDistance=true
    Module.bUseMinDistance=false
    Scalability Mode=::NewEnumerator1
    Scale Spawn Count=true
    UseLoopDelay=false
   )
   &SpawnRate(SpawnRate=180, Use Spawn Probability=false)
  ]
  particleSpawnScript [
   &InitializeParticle(
    Color=&RandomRangeLinearColor(Minimum=<rgba 0.4305 0.8048 1.5 1>, Maximum=<rgba 0.321 1.5 1.3035 1>)
    Lifetime Max=0.4
    Lifetime Min=0.2
    Lifetime Mode=::NewEnumerator1
    Sprite Size Mode=::NewEnumerator1
    Uniform Sprite Size Max=4
    Uniform Sprite Size Min=1
    UsePositionOffset=true
    Write Material Random=true
   )
   &SocketLocation(
    Interpolate Socket Sampling=true
    SamplePosOffset=false
    Socket Reader=$System.Socket Reader
    Socket Sampling Mode=::NewEnumerator2
    VelocitySpeedLimit=false
   )
   &ShapeLocation(MaskSpawnGroup=false, Override Seed=false, Sphere Radius=6, Transform Order:ENiagara_TransformOrder=::NewEnumerator1, UseOwnerScale=true) {module=ShapeLocation001}
   &AddVelocity(
    Module.Use Falloff Distance=false
    Override Seed=false
    Velocity Mode=::NewEnumerator1
    Velocity Speed=&RandomRangeFloat(Minimum=0, Maximum=50)
   ) {module=AddVelocity001}
  ]
  particleUpdateScript [
   &ParticleState()
   &AccelerationForce(Acceleration=<0 0 -100>)
   &Drag(Drag=10, Use Linear Drag=true, Use Rotational Drag=true)
   &CurlNoiseForce(Noise Strength=200, Pan Noise=false)
   &SolveForcesAndVelocity(Clamp Velocity=false, Limit Acceleration=false)
   &ScaleSpriteSize(Uniform Curve Sprite Scale=DataInterface() {propertyValues {Curve {keys [
        {tangentMode=RCTM_User, time=0, value=1}
        {arriveTangent=-0.0405, interpMode=RCIM_Cubic, leaveTangent=-0.0405, tangentMode=RCTM_User, time=0.7009, value=1}
        {arriveTangent=-8.0089, interpMode=RCIM_Cubic, leaveTangent=-8.0089, tangentMode=RCTM_User, time=1, value=0}
       ]}}})
   &ScaleColor(
    Scale Alpha=&FNTimeOfDayScalar(Night Scale=0.5, Morning Scale=0.75, Evening Scale=0.75)
    ScaleA=true
    ScaleRGB=true
   )
  ]
  rendererClasses [&NiagaraSpriteRendererProperties]
  renderers [{rendererClass=&NiagaraSpriteRendererProperties}]
 }
 TrailIce {
  emitterUpdateScript [
   &EmitterState()
   &FN_SpawnAnimTrail()
  ]
  particleSpawnScript [
   &FN_InitializeAnimTrailBoneIndecies()
   &FN_InitializeAnimTrail(AnimTrailWidth=1.75, Color=<rgba 0 2.2438 3 0.35>, Lifetime=0.22)
   &SystemLocation(MaskSpawnGroup=false)
  ]
  particleUpdateScript [
   &ScaleRibbonWidth(Ribbon Width Scale=&FloatFromCurve(FloatCurve=DataInterface() {propertyValues {Curve {keys [
         {arriveTangent=-1.2579, interpMode=RCIM_Cubic, time=0, value=1}
         {arriveTangent=-0.9664, interpMode=RCIM_Cubic, leaveTangent=-0.9664, tangentMode=RCTM_User, time=1, value=0}
        ]}}}))
   &ScaleColor(
    Scale Alpha=&Multiply_Float(
     A=&FloatFromCurve(FloatCurve=DataInterface() {propertyValues {Curve {keys [{interpMode=RCIM_Cubic, leaveTangent=-5.4253, tangentMode=RCTM_Break, time=0, value=1}, <at 0.4 0>]}}})
     B=&FNTimeOfDayScalar(Night Scale=0.75, Morning Scale=0.9, Evening Scale=0.9)
    )
    ScaleA=true
    ScaleRGB=true
   ) {module=ScaleColor002}
   &FN_UpdateAnimTrail()
  ]
  rendererClasses [&NiagaraRibbonRendererProperties]
  renderers [{rendererClass=&NiagaraRibbonRendererProperties}]
 }
 Trail_Bright_Blue {
  emitterUpdateScript [
   &EmitterState()
   &FN_SpawnAnimTrail()
  ]
  particleSpawnScript [
   &FN_InitializeAnimTrailBoneIndecies()
   &FN_InitializeAnimTrail(AnimTrailWidth=0.5, Color=<rgba 0.5 3.1855 5 0.5>, Lifetime=0.22)
   &SystemLocation(MaskSpawnGroup=false)
  ]
  particleUpdateScript [
   &ScaleRibbonWidth(Ribbon Width Scale=&FloatFromCurve(FloatCurve=DataInterface() {propertyValues {Curve {keys [
         {arriveTangent=-1.2579, interpMode=RCIM_Cubic, time=0, value=1}
         {arriveTangent=-0.9664, interpMode=RCIM_Cubic, leaveTangent=-0.9664, tangentMode=RCTM_User, time=1, value=0}
        ]}}}))
   &ScaleColor(
    Scale Alpha=&Multiply_Float(
     A=&FloatFromCurve(FloatCurve=DataInterface() {propertyValues {Curve {keys [
          {interpMode=RCIM_Cubic, leaveTangent=-2.3646, tangentMode=RCTM_Break, time=0, value=1}
          <at 0.6 0>
          {arriveTangent=-0.2139, interpMode=RCIM_Cubic, tangentMode=RCTM_Break, time=1, value=0}
         ]}}})
     B=&FNTimeOfDayScalar(Night Scale=0.75, Morning Scale=0.9, Evening Scale=0.9)
    )
    ScaleA=true
    ScaleRGB=true
   ) {module=ScaleColor002}
   &FN_UpdateAnimTrail()
  ]
  rendererClasses [&NiagaraRibbonRendererProperties]
  renderers [{rendererClass=&NiagaraRibbonRendererProperties}]
 }
}
summary {
 systemName=NS_Pickaxe_Crew_Trail_Ice
 emitters [
  {emitterName=Trail_Bright_Blue, rendererClasses [&NiagaraRibbonRendererProperties]}
  {emitterName=Sprite_Ice_Flakes, rendererClasses [&NiagaraSpriteRendererProperties]}
  {emitterName=TrailIce, rendererClasses [&NiagaraRibbonRendererProperties]}
 ]
 userVariables [User.RibbonIDOffset {type=NiagaraInt32}]
}
